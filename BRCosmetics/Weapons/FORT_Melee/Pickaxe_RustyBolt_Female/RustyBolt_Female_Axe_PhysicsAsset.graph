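# Physics asset for the female Rusty Bolt pickaxe, giving the two-handed axe's dangling cloth strips physical bodies.
# A long capsule on the handle bone anchors two mirrored chains, left and right, each made of five small cloth capsules.
# Constraints link each segment to the one above with moderate swing limits and tight twist, slightly looser at the handle attachment, so the strips bend freely without spinning.
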PhysicsAsset /BRCosmetics/Weapons/FORT_Melee/Pickaxe_RustyBolt_Female/RustyBolt_Female_Axe_PhysicsAsset
bodies {
 Melee_TwoHanded_Handle [{shapeType=Capsule, length=68.3692, radius=3.336, center=<0.564 -0.7669 0>, rotation=<rot 0 -90 0>}]
 cloth_0_lf [{shapeType=Capsule, length=1.25, radius=2.75, shapeName=cloth_0_lf_capsule, center=<0 0 -3>, rotation=<rot 0 0 89.9999>}]
 cloth_0_rt [{shapeType=Capsule, length=1.25, radius=2.75, shapeName=cloth_0_rt_capsule, center=<0 0 -3>, rotation=<rot 0 0 89.9999>}]
 cloth_1_lf [{shapeType=Capsule, length=1, radius=1.5, shapeName=cloth_1_lf_capsule, center=<1.2684 0 0>, rotation=<rot 0 0 89.9999>}]
 cloth_1_rt [{shapeType=Capsule, length=1, radius=1.5, shapeName=cloth_1_rt_capsule, center=<1.2684 0 0>, rotation=<rot 0 0 89.9999>}]
 cloth_2_lf [{shapeType=Capsule, length=2, radius=1.5, shapeName=cloth_2_lf_capsule, center=<1.2684 0 0>, rotation=<rot 0 0 89.9999>}]
 cloth_2_rt [{shapeType=Capsule, length=2, radius=1.5, shapeName=cloth_2_rt_capsule, center=<1.2684 0 0>, rotation=<rot 0 0 89.9999>}]
 cloth_3_lf [{shapeType=Capsule, length=2, radius=1.5, shapeName=cloth_3_lf_capsule, center=<1.2684 0 0>, rotation=<rot 0 0 89.9999>}]
 cloth_3_rt [{shapeType=Capsule, length=2, radius=1.5, shapeName=cloth_3_rt_capsule, center=<1.2684 0 0>, rotation=<rot 0 0 89.9999>}]
 cloth_4_lf [{shapeType=Capsule, length=2, radius=1.5, shapeName=cloth_4_lf_capsule, center=<1.2684 0 0>, rotation=<rot 0 0 89.9999>}]
 cloth_4_rt [{shapeType=Capsule, length=2, radius=1.5, shapeName=cloth_4_rt_capsule, center=<1.2684 0 0>, rotation=<rot 0 0 89.9999>}]
}
constraints [
 {bone1Name=cloth_0_rt, bone2Name=Melee_TwoHanded_Handle, swing1LimitDegrees=45, swing2LimitDegrees=45, twistLimitDegrees=15}
 {bone1Name=cloth_1_rt, bone2Name=cloth_0_rt, swing1LimitDegrees=45, swing2LimitDegrees=45, twistLimitDegrees=5}
 {bone1Name=cloth_2_rt, bone2Name=cloth_1_rt, swing1LimitDegrees=45, swing2LimitDegrees=45, twistLimitDegrees=5}
 {bone1Name=cloth_3_rt, bone2Name=cloth_2_rt, swing1LimitDegrees=45, swing2LimitDegrees=45, twistLimitDegrees=5}
 {bone1Name=cloth_4_rt, bone2Name=cloth_3_rt, swing1LimitDegrees=45, swing2LimitDegrees=45, twistLimitDegrees=5}
 {bone1Name=cloth_0_lf, bone2Name=Melee_TwoHanded_Handle, swing1LimitDegrees=45, swing2LimitDegrees=45, twistLimitDegrees=15}
 {bone1Name=cloth_1_lf, bone2Name=cloth_0_lf, swing1LimitDegrees=45, swing2LimitDegrees=45, twistLimitDegrees=5}
 {bone1Name=cloth_2_lf, bone2Name=cloth_1_lf, swing1LimitDegrees=45, swing2LimitDegrees=45, twistLimitDegrees=5}
 {bone1Name=cloth_3_lf, bone2Name=cloth_2_lf, swing1LimitDegrees=45, swing2LimitDegrees=45, twistLimitDegrees=5}
 {bone1Name=cloth_4_lf, bone2Name=cloth_3_lf, swing1LimitDegrees=45, swing2LimitDegrees=45, twistLimitDegrees=5}
]
