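# Anim Blueprint for the RustyBolt Female pickaxe that animates a staggered equip motion on its spine bones.
# The EventGraph reads the owner's weapon anim type and front-end flag, setting per-bone alphas to 1 after successive delays from spine_0 to spine_6.
# Interpolated alphas drive alternating ±45° Modify Bone rotations, feeding a handle-space RigidBody simulation.
# AnimDynamics cloth and 90° spine chains sit disconnected from the output.
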AnimBlueprint /BRCosmetics/Weapons/FORT_Melee/Pickaxe_RustyBolt_Female/RustyBolt_Female_Axe_Skeleton_AnimBlueprint
target_skeleton=&/RustyBolt_Female_Axe_Skeleton
graphs [
 AnimGraph {nodes [
   n1 = Anim Dynamics - Bone: cloth_0_lf(Alpha=0.75, ExternalForce="0,0,0") {
    angularDampingOverride=0.9999
    angularSpringConstant=100
    bGravityOverrideInSimSpace=true
    bOverrideAngularDamping=true
    alphaBoolBlend {blendOption=Linear}
    boundBone {boneName=cloth_0_lf}
    componentAppliedLinearAccClamp=<5000 5000 5000>
    componentLinearAccScale=<0.5 0.5 0.5>
    componentLinearVelScale=<0.5 0.5 0.5>
    gravityOverride {x=-980}
    physicsBodyDefinitions [{
      sphereCollisionRadius=0
      boundBone {boneName=cloth_0_lf}
      boxExtents=<12 22 22>
      constraintSetup {angularLimitsMax=<10 10 -5>, angularLimitsMin=<-10 -57.6001 -80>, angularTarget=<0 0 0>}
      localJointOffset=<6 0 0>
     }]
   }
   n2 = Anim Dynamics - Bone: cloth_1_lf(ComponentPose=@n1, Alpha=0.75, ExternalForce="0,0,0") {
    angularDampingOverride=0.9999
    angularSpringConstant=100
    bGravityOverrideInSimSpace=true
    bOverrideAngularDamping=true
    alphaBoolBlend {blendOption=Linear}
    boundBone {boneName=cloth_1_lf}
    componentAppliedLinearAccClamp=<5000 5000 5000>
    componentLinearAccScale=<0.5 0.5 0.5>
    componentLinearVelScale=<0.5 0.5 0.5>
    gravityOverride {x=-980}
    physicsBodyDefinitions [{
      sphereCollisionRadius=0
      boundBone {boneName=cloth_1_lf}
      boxExtents=<12 22 22>
      constraintSetup {angularLimitsMax=<10 10 -5>, angularLimitsMin=<-10 -10 -80>, angularTarget=<0 0 0>}
      localJointOffset=<6 0 0>
     }]
   }
   n3 = Anim Dynamics - Bone: cloth_2_lf(ComponentPose=@n2, Alpha=0.75, ExternalForce="0,0,0") {
    angularDampingOverride=0.9999
    angularSpringConstant=100
    bGravityOverrideInSimSpace=true
    bOverrideAngularDamping=true
    alphaBoolBlend {blendOption=Linear}
    boundBone {boneName=cloth_2_lf}
    componentAppliedLinearAccClamp=<2000 2000 3000>
    componentLinearAccScale=<0.2 0.2 0.2>
    componentLinearVelScale=<0.2 0.2 0.2>
    gravityOverride {x=-980}
    physicsBodyDefinitions [{
      sphereCollisionRadius=0
      boundBone {boneName=cloth_2_lf}
      boxExtents=<12 22 22>
      constraintSetup {angularLimitsMax=<5 10 0>, angularLimitsMin=<-5 -10 -80>, angularTarget=<0 0 0>}
      localJointOffset=<6 0 0>
     }]
   }
   n4 = Anim Dynamics - Bone: cloth_3_lf(ComponentPose=@n3, Alpha=0.75, ExternalForce="0,0,0") {
    angularDampingOverride=0.9999
    angularSpringConstant=100
    bGravityOverrideInSimSpace=true
    bOverrideAngularDamping=true
    alphaBoolBlend {blendOption=Linear}
    boundBone {boneName=cloth_3_lf}
    componentAppliedLinearAccClamp=<2000 2000 3000>
    componentLinearAccScale=<0.2 0.2 0.2>
    componentLinearVelScale=<0.2 0.2 0.2>
    gravityOverride {x=-980}
    physicsBodyDefinitions [{
      sphereCollisionRadius=0
      boundBone {boneName=cloth_3_lf}
      boxExtents=<12 22 22>
      constraintSetup {angularLimitsMax=<5 10 0>, angularLimitsMin=<-5 -10 -80>, angularTarget=<0 0 0>}
      localJointOffset=<6 0 0>
     }]
   }
   n5 = Anim Dynamics - Bone: cloth_4_lf(ComponentPose=@n4, Alpha=0.75, ExternalForce="0,0,0") {
    angularDampingOverride=0.9999
    angularSpringConstant=100
    bGravityOverrideInSimSpace=true
    bOverrideAngularDamping=true
    alphaBoolBlend {blendOption=Linear}
    boundBone {boneName=cloth_4_lf}
    componentAppliedLinearAccClamp=<2000 2000 3000>
    componentLinearAccScale=<0.2 0.2 0.2>
    componentLinearVelScale=<0.2 0.2 0.2>
    gravityOverride {x=-980}
    physicsBodyDefinitions [{
      sphereCollisionRadius=0
      boundBone {boneName=cloth_4_lf}
      boxExtents=<12 22 22>
      constraintSetup {angularLimitsMax=<5 20 0>, angularLimitsMin=<-5 -20 -80>, angularTarget=<0 0 0>}
      localJointOffset=<6 0 0>
     }]
   }
   n6 = Anim Dynamics - Bone: cloth_5_lf(ComponentPose=@n5, Alpha=0.75, ExternalForce="0,0,0") {
    angularDampingOverride=0.9999
    angularSpringConstant=100
    bGravityOverrideInSimSpace=true
    bOverrideAngularDamping=true
    alphaBoolBlend {blendOption=Linear}
    boundBone {boneName=cloth_5_lf}
    componentAppliedLinearAccClamp=<2000 2000 3000>
    componentLinearAccScale=<0.2 0.2 0.2>
    componentLinearVelScale=<0.2 0.2 0.2>
    gravityOverride {x=-980}
    physicsBodyDefinitions [{
      sphereCollisionRadius=0
      boundBone {boneName=cloth_5_lf}
      boxExtents=<12 22 22>
      constraintSetup {angularLimitsMax=<5 10 0>, angularLimitsMin=<-5 -10 -80>, angularTarget=<0 0 0>}
      localJointOffset=<6 0 0>
     }]
   }
   n7 = Anim Dynamics - Bone: cloth_6_lf(ComponentPose=@n6, Alpha=0.75, ExternalForce="0,0,0") {
    angularDampingOverride=0.9999
    angularSpringConstant=100
    bGravityOverrideInSimSpace=true
    bOverrideAngularDamping=true
    alphaBoolBlend {blendOption=Linear}
    boundBone {boneName=cloth_6_lf}
    componentAppliedLinearAccClamp=<2000 2000 3000>
    componentLinearAccScale=<0.2 0.2 0.2>
    componentLinearVelScale=<0.2 0.2 0.2>
    gravityOverride {x=-980}
    physicsBodyDefinitions [{
      sphereCollisionRadius=0
      boundBone {boneName=cloth_6_lf}
      boxExtents=<12 22 22>
      constraintSetup {angularLimitsMax=<5 20 0>, angularLimitsMin=<-5 -20 -80>, angularTarget=<0 0 0>}
      localJointOffset=<6 0 0>
     }]
   }
   n8 = Anim Dynamics - Bone: cloth_0_rt(ComponentPose=@n7, Alpha=0.75, ExternalForce="0,0,0") {
    angularDampingOverride=0.999
    angularSpringConstant=100
    bGravityOverrideInSimSpace=true
    bOverrideAngularDamping=true
    alphaBoolBlend {blendOption=Linear}
    boundBone {boneName=cloth_0_rt}
    componentAppliedLinearAccClamp=<5000 5000 5000>
    componentLinearAccScale=<0.3 0.3 0.3>
    componentLinearVelScale=<0.3 0.3 0.3>
    gravityOverride {x=-980}
    physicsBodyDefinitions [{
      sphereCollisionRadius=0
      boundBone {boneName=cloth_0_rt}
      boxExtents=<12 22 22>
      constraintSetup {angularLimitsMax=<10 10 -5>, angularLimitsMin=<-10 -10 -80>, angularTarget=<0 0 0>}
      localJointOffset=<6 0 0>
     }]
   }
   n9 = Anim Dynamics - Bone: cloth_1_rt(ComponentPose=@n8, Alpha=0.75, ExternalForce="0,0,0") {
    angularDampingOverride=0.999
    angularSpringConstant=100
    bGravityOverrideInSimSpace=true
    bOverrideAngularDamping=true
    alphaBoolBlend {blendOption=Linear}
    boundBone {boneName=cloth_1_rt}
    componentAppliedLinearAccClamp=<5000 5000 5000>
    componentLinearAccScale=<0.3 0.3 0.3>
    componentLinearVelScale=<0.3 0.3 0.3>
    gravityOverride {x=-980}
    physicsBodyDefinitions [{
      sphereCollisionRadius=0
      boundBone {boneName=cloth_1_rt}
      boxExtents=<12 22 22>
      constraintSetup {angularLimitsMax=<10 10 -5>, angularLimitsMin=<-10 -10 -80>, angularTarget=<0 0 0>}
      localJointOffset=<6 0 0>
     }]
   }
   n10 = Anim Dynamics - Bone: cloth_2_rt(ComponentPose=@n9, Alpha=0.75, ExternalForce="0,0,0") {
    angularDampingOverride=0.999
    angularSpringConstant=100
    bGravityOverrideInSimSpace=true
    bOverrideAngularDamping=true
    alphaBoolBlend {blendOption=Linear}
    boundBone {boneName=cloth_2_rt}
    componentAppliedLinearAccClamp=<2000 2000 3000>
    componentLinearAccScale=<0.2 0.2 0.2>
    componentLinearVelScale=<0.2 0.2 0.2>
    gravityOverride {x=-980}
    physicsBodyDefinitions [{
      sphereCollisionRadius=0
      boundBone {boneName=cloth_2_rt}
      boxExtents=<12 22 22>
      constraintSetup {angularLimitsMax=<5 10 0>, angularLimitsMin=<-5 -10 -80>, angularTarget=<0 0 0>}
      localJointOffset=<6 0 0>
     }]
   }
   n11 = Anim Dynamics - Bone: cloth_3_rt(ComponentPose=@n10, Alpha=0.75, ExternalForce="0,0,0") {
    angularDampingOverride=0.999
    angularSpringConstant=100
    bGravityOverrideInSimSpace=true
    bOverrideAngularDamping=true
    alphaBoolBlend {blendOption=Linear}
    boundBone {boneName=cloth_3_rt}
    componentAppliedLinearAccClamp=<2000 2000 3000>
    componentLinearAccScale=<0.2 0.2 0.2>
    componentLinearVelScale=<0.2 0.2 0.2>
    gravityOverride {x=-980}
    physicsBodyDefinitions [{
      sphereCollisionRadius=0
      boundBone {boneName=cloth_3_rt}
      boxExtents=<12 22 22>
      constraintSetup {angularLimitsMax=<5 10 0>, angularLimitsMin=<-5 -10 -80>, angularTarget=<0 0 0>}
      localJointOffset=<6 0 0>
     }]
   }
   n12 = Anim Dynamics - Bone: cloth_4_rt(ComponentPose=@n11, Alpha=0.75, ExternalForce="0,0,0") {
    angularDampingOverride=0.999
    angularSpringConstant=100
    bGravityOverrideInSimSpace=true
    bOverrideAngularDamping=true
    alphaBoolBlend {blendOption=Linear}
    boundBone {boneName=cloth_4_rt}
    componentAppliedLinearAccClamp=<2000 2000 3000>
    componentLinearAccScale=<0.2 0.2 0.2>
    componentLinearVelScale=<0.2 0.2 0.2>
    gravityOverride {x=-980}
    physicsBodyDefinitions [{
      sphereCollisionRadius=0
      boundBone {boneName=cloth_4_rt}
      boxExtents=<12 22 22>
      constraintSetup {angularLimitsMax=<5 20 0>, angularLimitsMin=<-5 -20 -80>, angularTarget=<0 0 0>}
      localJointOffset=<6 0 0>
     }]
   }
   n13 = Anim Dynamics - Bone: cloth_5_rt(ComponentPose=@n12, Alpha=0.75, ExternalForce="0,0,0") {
    angularDampingOverride=0.999
    angularSpringConstant=100
    bGravityOverrideInSimSpace=true
    bOverrideAngularDamping=true
    alphaBoolBlend {blendOption=Linear}
    boundBone {boneName=cloth_5_rt}
    componentAppliedLinearAccClamp=<2000 2000 3000>
    componentLinearAccScale=<0.2 0.2 0.2>
    componentLinearVelScale=<0.2 0.2 0.2>
    gravityOverride {x=-980}
    physicsBodyDefinitions [{
      sphereCollisionRadius=0
      boundBone {boneName=cloth_5_rt}
      boxExtents=<12 22 22>
      constraintSetup {angularLimitsMax=<5 10 0>, angularLimitsMin=<-5 -10 -80>, angularTarget=<0 0 0>}
      localJointOffset=<6 0 0>
     }]
   }
   Anim Dynamics - Bone: cloth_6_rt(ComponentPose=@n13, Alpha=0.75, ExternalForce="0,0,0") {
    angularDampingOverride=0.999
    angularSpringConstant=100
    bGravityOverrideInSimSpace=true
    bOverrideAngularDamping=true
    alphaBoolBlend {blendOption=Linear}
    boundBone {boneName=cloth_6_rt}
    componentAppliedLinearAccClamp=<2000 2000 3000>
    componentLinearAccScale=<0.2 0.2 0.2>
    componentLinearVelScale=<0.2 0.2 0.2>
    gravityOverride {x=-980}
    physicsBodyDefinitions [{
      sphereCollisionRadius=0
      boundBone {boneName=cloth_6_rt}
      boxExtents=<12 22 22>
      constraintSetup {angularLimitsMax=<5 20 0>, angularLimitsMin=<-5 -20 -80>, angularTarget=<0 0 0>}
      localJointOffset=<6 0 0>
     }]
   }
   n14 = "Transform (Modify) Bone - Bone: spine_6"(Alpha=1, Translation="0,0,0", Rotation="0,90,0", Scale="1,1,1") {
    rotationMode=BMM_Additive
    scaleMode=BMM_Replace
    scaleSpace=BCS_BoneSpace
    translationMode=BMM_Additive
    alphaBoolBlend {blendOption=Linear}
    alphaScaleBiasClamp {bInterpResult=true, interpSpeedDecreasing=5, interpSpeedIncreasing=5}
    boneToModify {boneName=spine_6}
    rotation {yaw=90}
   }
   n15 = "Transform (Modify) Bone - Bone: spine_4"(ComponentPose=@n14, Alpha=1, Translation="0,0,0", Rotation="0,90,0", Scale="1,1,1") {
    rotationMode=BMM_Additive
    scaleMode=BMM_Replace
    scaleSpace=BCS_BoneSpace
    translationMode=BMM_Additive
    alphaBoolBlend {blendOption=Linear}
    alphaScaleBiasClamp {bInterpResult=true, interpSpeedDecreasing=5, interpSpeedIncreasing=5}
    boneToModify {boneName=spine_4}
    rotation {yaw=90}
   }
   n16 = "Transform (Modify) Bone - Bone: spine_2"(ComponentPose=@n15, Alpha=1, Translation="0,0,0", Rotation="0,90,0", Scale="1,1,1") {
    rotationMode=BMM_Additive
    scaleMode=BMM_Replace
    scaleSpace=BCS_BoneSpace
    translationMode=BMM_Additive
    alphaBoolBlend {blendOption=Linear}
    alphaScaleBiasClamp {bInterpResult=true, interpSpeedDecreasing=5, interpSpeedIncreasing=5}
    boneToModify {boneName=spine_2}
    rotation {yaw=90}
   }
   "Transform (Modify) Bone - Bone: spine_0"(ComponentPose=@n16, Alpha=1, Translation="0,0,0", Rotation="0,90,0", Scale="1,1,1") {
    rotationMode=BMM_Additive
    scaleMode=BMM_Replace
    scaleSpace=BCS_BoneSpace
    translationMode=BMM_Additive
    alphaBoolBlend {blendOption=Linear}
    alphaScaleBiasClamp {bInterpResult=true, interpSpeedDecreasing=5, interpSpeedIncreasing=5}
    boneToModify {boneName=spine_0}
    rotation {yaw=90}
   }
   n17 = "Transform (Modify) Bone - Bone: spine_6"(
    ComponentPose=Local To Component(LocalPose=Input Pose())
    Alpha=Get equipAlpha() -> (equipAlpha:double)
    Translation="0,0,0"
    Rotation="0,45,0"
    Scale="1,1,1"
   ) {
    rotationMode=BMM_Additive
    scaleMode=BMM_Replace
    scaleSpace=BCS_BoneSpace
    translationMode=BMM_Additive
    alphaBoolBlend {blendOption=Linear}
    alphaScaleBiasClamp {bInterpResult=true, interpSpeedDecreasing=5, interpSpeedIncreasing=5}
    boneToModify {boneName=spine_6}
    rotation {yaw=45}
   }
   n18 = "Transform (Modify) Bone - Bone: spine_4"(ComponentPose=@n17, Alpha=Get equipAlpha_2() -> (equipAlpha_2:double), Translation="0,0,0", Rotation="0,45,0", Scale="1,1,1") {
    rotationMode=BMM_Additive
    scaleMode=BMM_Replace
    scaleSpace=BCS_BoneSpace
    translationMode=BMM_Additive
    alphaBoolBlend {blendOption=Linear}
    alphaScaleBiasClamp {bInterpResult=true, interpSpeedDecreasing=5, interpSpeedIncreasing=5}
    boneToModify {boneName=spine_4}
    rotation {yaw=45}
   }
   n19 = "Transform (Modify) Bone - Bone: spine_2"(ComponentPose=@n18, Alpha=Get equipAlpha_4() -> (equipAlpha_4:double), Translation="0,0,0", Rotation="0,45,0", Scale="1,1,1") {
    rotationMode=BMM_Additive
    scaleMode=BMM_Replace
    scaleSpace=BCS_BoneSpace
    translationMode=BMM_Additive
    alphaBoolBlend {blendOption=Linear}
    alphaScaleBiasClamp {bInterpResult=true, interpSpeedDecreasing=5, interpSpeedIncreasing=5}
    boneToModify {boneName=spine_2}
    rotation {yaw=45}
   }
   n20 = "Transform (Modify) Bone - Bone: spine_0"(ComponentPose=@n19, Alpha=Get equipAlpha_6() -> (equipAlpha_6:double), Translation="0,0,0", Rotation="0,45,0", Scale="1,1,1") {
    rotationMode=BMM_Additive
    scaleMode=BMM_Replace
    scaleSpace=BCS_BoneSpace
    translationMode=BMM_Additive
    alphaBoolBlend {blendOption=Linear}
    alphaScaleBiasClamp {bInterpResult=true, interpSpeedDecreasing=5, interpSpeedIncreasing=5}
    boneToModify {boneName=spine_0}
    rotation {yaw=45}
   }
   n21 = "Transform (Modify) Bone - Bone: spine_5"(ComponentPose=@n20, Alpha=Get equipAlpha_1() -> (equipAlpha_1:double), Translation="0,0,0", Rotation="0,-45,0", Scale="1,1,1") {
    rotationMode=BMM_Additive
    scaleMode=BMM_Replace
    scaleSpace=BCS_BoneSpace
    translationMode=BMM_Additive
    alphaBoolBlend {blendOption=Linear}
    alphaScaleBiasClamp {bInterpResult=true, interpSpeedDecreasing=5, interpSpeedIncreasing=5}
    boneToModify {boneName=spine_5}
    rotation {yaw=-45}
   }
   n22 = "Transform (Modify) Bone - Bone: spine_3"(ComponentPose=@n21, Alpha=Get equipAlpha_3() -> (equipAlpha_3:double), Translation="0,0,0", Rotation="0,-45,0", Scale="1,1,1") {
    rotationMode=BMM_Additive
    scaleMode=BMM_Replace
    scaleSpace=BCS_BoneSpace
    translationMode=BMM_Additive
    alphaBoolBlend {blendOption=Linear}
    alphaScaleBiasClamp {bInterpResult=true, interpSpeedDecreasing=5, interpSpeedIncreasing=5}
    boneToModify {boneName=spine_3}
    rotation {yaw=-45}
   }
   n23 = "Transform (Modify) Bone - Bone: spine_1"(ComponentPose=@n22, Alpha=Get equipAlpha_5() -> (equipAlpha_5:double), Translation="0,0,0", Rotation="0,-45,0", Scale="1,1,1") {
    rotationMode=BMM_Additive
    scaleMode=BMM_Replace
    scaleSpace=BCS_BoneSpace
    translationMode=BMM_Additive
    alphaBoolBlend {blendOption=Linear}
    alphaScaleBiasClamp {bInterpResult=true, interpSpeedDecreasing=5, interpSpeedIncreasing=5}
    boneToModify {boneName=spine_1}
    rotation {yaw=-45}
   }
   n24 = "Transform (Modify) Bone - Bone: spine_6"(ComponentPose=@n23, Alpha=Get equipAlpha_FE() -> (equipAlpha_FE:double), Translation="0,0,0", Rotation="0,45,0", Scale="1,1,1") {
    rotationMode=BMM_Additive
    scaleMode=BMM_Replace
    scaleSpace=BCS_BoneSpace
    translationMode=BMM_Additive
    alphaBoolBlend {blendOption=Linear}
    alphaScaleBiasClamp {bInterpResult=true, interpSpeedDecreasing=5, interpSpeedIncreasing=5}
    boneToModify {boneName=spine_6}
    rotation {yaw=45}
   }
   n25 = "Transform (Modify) Bone - Bone: spine_4"(ComponentPose=@n24, Alpha=Get equipAlpha_FE_02() -> (equipAlpha_FE_02:double), Translation="0,0,0", Rotation="0,45,0", Scale="1,1,1") {
    rotationMode=BMM_Additive
    scaleMode=BMM_Replace
    scaleSpace=BCS_BoneSpace
    translationMode=BMM_Additive
    alphaBoolBlend {blendOption=Linear}
    alphaScaleBiasClamp {bInterpResult=true, interpSpeedDecreasing=5, interpSpeedIncreasing=5}
    boneToModify {boneName=spine_4}
    rotation {yaw=45}
   }
   n26 = "Transform (Modify) Bone - Bone: spine_2"(ComponentPose=@n25, Alpha=Get equipAlpha_FE_04() -> (equipAlpha_FE_04:double), Translation="0,0,0", Rotation="0,45,0", Scale="1,1,1") {
    rotationMode=BMM_Additive
    scaleMode=BMM_Replace
    scaleSpace=BCS_BoneSpace
    translationMode=BMM_Additive
    alphaBoolBlend {blendOption=Linear}
    alphaScaleBiasClamp {bInterpResult=true, interpSpeedDecreasing=5, interpSpeedIncreasing=5}
    boneToModify {boneName=spine_2}
    rotation {yaw=45}
   }
   n27 = "Transform (Modify) Bone - Bone: spine_0"(ComponentPose=@n26, Alpha=Get equipAlpha_FE_06() -> (equipAlpha_FE_06:double), Translation="0,0,0", Rotation="0,45,0", Scale="1,1,1") {
    rotationMode=BMM_Additive
    scaleMode=BMM_Replace
    scaleSpace=BCS_BoneSpace
    translationMode=BMM_Additive
    alphaBoolBlend {blendOption=Linear}
    alphaScaleBiasClamp {bInterpResult=true, interpSpeedDecreasing=5, interpSpeedIncreasing=5}
    boneToModify {boneName=spine_0}
    rotation {yaw=45}
   }
   n28 = "Transform (Modify) Bone - Bone: spine_5"(ComponentPose=@n27, Alpha=Get equipAlpha_FE_01() -> (equipAlpha_FE_01:double), Translation="0,0,0", Rotation="0,-45,0", Scale="1,1,1") {
    rotationMode=BMM_Additive
    scaleMode=BMM_Replace
    scaleSpace=BCS_BoneSpace
    translationMode=BMM_Additive
    alphaBoolBlend {blendOption=Linear}
    alphaScaleBiasClamp {bInterpResult=true, interpSpeedDecreasing=5, interpSpeedIncreasing=5}
    boneToModify {boneName=spine_5}
    rotation {yaw=-45}
   }
   n29 = "Transform (Modify) Bone - Bone: spine_3"(ComponentPose=@n28, Alpha=Get equipAlpha_FE_03() -> (equipAlpha_FE_03:double), Translation="0,0,0", Rotation="0,-45,0", Scale="1,1,1") {
    rotationMode=BMM_Additive
    scaleMode=BMM_Replace
    scaleSpace=BCS_BoneSpace
    translationMode=BMM_Additive
    alphaBoolBlend {blendOption=Linear}
    alphaScaleBiasClamp {bInterpResult=true, interpSpeedDecreasing=5, interpSpeedIncreasing=5}
    boneToModify {boneName=spine_3}
    rotation {yaw=-45}
   }
   n30 = "Transform (Modify) Bone - Bone: spine_1"(ComponentPose=@n29, Alpha=Get equipAlpha_FE_05() -> (equipAlpha_FE_05:double), Translation="0,0,0", Rotation="0,-45,0", Scale="1,1,1") {
    rotationMode=BMM_Additive
    scaleMode=BMM_Replace
    scaleSpace=BCS_BoneSpace
    translationMode=BMM_Additive
    alphaBoolBlend {blendOption=Linear}
    alphaScaleBiasClamp {bInterpResult=true, interpSpeedDecreasing=5, interpSpeedIncreasing=5}
    boneToModify {boneName=spine_1}
    rotation {yaw=-45}
   }
   n31 = RigidBody(ComponentPose=@n30) {
    bClampLinearTranslationLimitToRefPose=true
    bOverrideWorldGravity=true
    overridePhysicsAsset=&/RustyBolt_Female_Axe_PhysicsAsset
    simulationSpace=BaseBoneSpace
    alphaBoolBlend {blendOption=Linear}
    alphaScaleBiasClamp {bInterpResult=true, interpSpeedDecreasing=15, interpSpeedIncreasing=15}
    baseBoneRef {boneName=Melee_TwoHanded_Handle}
    componentAppliedLinearAccClamp=<5000 5000 5000>
    componentLinearAccScale=<0.5 0.5 0.5>
    componentLinearVelScale=<0.5 0.5 0.5>
    overrideWorldGravity {z=-980}
   }
   Output Pose(Result=Component To Local(ComponentPose=@n31)) {layerGroup=None}
   SelectFloat(A=0.5, B=1, bPickA=Get bIsPlayingMeleeAnim())
  ]}
 EventGraph {nodes [
   Event BlueprintUpdateAnimation()
   > n1 = Sequence!(execute) -> (
    then_0
    then_1
    then_2:Exec
    then_3:Exec
    then_4:Exec
    then_5:Exec
    then_6:Exec
    then_7:Exec
    then_8:Exec
    then_9:Exec
    then_10:Exec
    then_11:Exec
    then_12:Exec
    then_13:Exec
    then_14:Exec
   )
   > n2 = Cast To FortPlayerAnimInstance(Object=GetAnimInstance(self=GetOwnerSkeletalMeshComponent()))
   > n3 = Switch on EFortWeaponCoreAnimation!(execute, Selection:EFortWeaponCoreAnimation Enum=Get WeaponCoreAnim(self=@n2)) -> (
    Melee:Exec
    Pistol:Exec
    Shotgun:Exec
    PaperBlueprint:Exec
    Rifle:Exec
    MeleeOneHand:Exec
    MachinePistol:Exec
    RocketLauncher:Exec
    GrenadeLauncher:Exec
    GoingCommando:Exec
    AssaultRifle:Exec
    TacticalShotgun:Exec
    SniperRifle:Exec
    TrapPlacement:Exec
    ShoulderLauncher:Exec
    AbilityDecoTool:Exec
    Crossbow:Exec
    C4:Exec
    RemoteControl:Exec
    DualWield:Exec
    AR_BullPup:Exec
    AR_ForwardGrip:Exec
    MedPackPaddles:Exec
    SMG_P90:Exec
    AR_DrumGun:Exec
    Consumable_Small:Exec
    Consumable_Large:Exec
    Balloon:Exec
    MountedTurret:Exec
    CreativeTool:Exec
    ExplosiveBow:Exec
    AshtonIndigo:Exec
    AshtonChicago:Exec
    MeleeDualWield:Exec
    Unarmed:Exec
   )
   Set isMelee(
    execute=[
     @n3.AR_BullPup
     @n3.AR_DrumGun
     @n3.AR_ForwardGrip
     @n3.AbilityDecoTool
     @n3.AshtonChicago
     @n3.AshtonIndigo
     @n3.AssaultRifle
     @n3.Balloon
     @n3.C4
     @n3.Consumable_Large
     @n3.Consumable_Small
     @n3.CreativeTool
     @n3.Crossbow
     @n3.DualWield
     @n3.ExplosiveBow
     @n3.GoingCommando
     @n3.GrenadeLauncher
     @n3.MachinePistol
     @n3.MedPackPaddles
     @n3.MountedTurret
     @n3.PaperBlueprint
     @n3.Pistol
     @n3.RemoteControl
     @n3.Rifle
     @n3.RocketLauncher
     @n3.SMG_P90
     @n3.Shotgun
     @n3.ShoulderLauncher
     @n3.SniperRifle
     @n3.TacticalShotgun
     @n3.TrapPlacement
    ]
    isMelee:bool=false
   ) -> (Output_Get:bool)
   Set isMelee(execute=[@n3.Melee, @n3.MeleeDualWield, @n3.MeleeOneHand], isMelee=true)
   n4 = Branch(execute=@n1.then_1, Condition=Get isMelee() -> (isMelee:bool))
   > Do Once(Reset=@n5.then)
   > Set equipAlpha(equipAlpha:double=0) -> (Output_Get:double)
   > Delay(Duration=0.7)
   > Set equipAlpha(equipAlpha=1)
   n5 = Set equipAlpha(execute=@n4.else, equipAlpha=0)
   n6 = Get bIsInFrontEnd()
   n7 = Branch(execute=@n1.then_10, Condition=@n6)
   Delay(execute=[@n7.then, @n28.then], Duration=0.5)
   > Set equipAlpha_FE_04(equipAlpha_FE_04:double=1) -> (Output_Get:double)
   n8 = Branch(execute=@n1.then_11, Condition=@n6)
   Delay(execute=[@n8.then, @n29.then], Duration=0.55)
   > Set equipAlpha_FE_03(equipAlpha_FE_03:double=1) -> (Output_Get:double)
   n9 = Branch(execute=@n1.then_12, Condition=@n6)
   Delay(execute=[@n9.then, @n30.then], Duration=0.6)
   > Set equipAlpha_FE_02(equipAlpha_FE_02:double=1) -> (Output_Get:double)
   n10 = Branch(execute=@n1.then_13, Condition=@n6)
   Delay(execute=[@n10.then, @n31.then], Duration=0.65)
   > Set equipAlpha_FE_01(equipAlpha_FE_01:double=1) -> (Output_Get:double)
   n11 = Branch(execute=@n1.then_14, Condition=@n6)
   Delay(execute=[@n11.then, @n32.then], Duration=0.7)
   > Set equipAlpha_FE(equipAlpha_FE:double=1) -> (Output_Get:double)
   n12 = Branch(execute=@n1.then_2, Condition=Get isMelee())
   > Do Once(Reset=@n13.then)
   > Set equipAlpha_1(equipAlpha_1:double=0) -> (Output_Get:double)
   > Delay(Duration=0.65)
   > Set equipAlpha_1(equipAlpha_1=1)
   n13 = Set equipAlpha_1(execute=@n12.else, equipAlpha_1=0)
   n14 = Branch(execute=@n1.then_3, Condition=Get isMelee())
   > Do Once(Reset=@n15.then)
   > Set equipAlpha_2(equipAlpha_2:double=0) -> (Output_Get:double)
   > Delay(Duration=0.6)
   > Set equipAlpha_2(equipAlpha_2=1)
   n15 = Set equipAlpha_2(execute=@n14.else, equipAlpha_2=0)
   n16 = Branch(execute=@n1.then_4, Condition=Get isMelee())
   > Do Once(Reset=@n17.then)
   > Set equipAlpha_3(equipAlpha_3:double=0) -> (Output_Get:double)
   > Delay(Duration=0.55)
   > Set equipAlpha_3(equipAlpha_3=1)
   n17 = Set equipAlpha_3(execute=@n16.else, equipAlpha_3=0)
   n18 = Branch(execute=@n1.then_5, Condition=Get isMelee())
   > Do Once(Reset=@n19.then)
   > Set equipAlpha_4(equipAlpha_4:double=0) -> (Output_Get:double)
   > Delay(Duration=0.5)
   > Set equipAlpha_4(equipAlpha_4=1)
   n19 = Set equipAlpha_4(execute=@n18.else, equipAlpha_4=0)
   n20 = Branch(execute=@n1.then_6, Condition=Get isMelee())
   > Do Once(Reset=@n21.then)
   > Set equipAlpha_5(equipAlpha_5:double=0) -> (Output_Get:double)
   > Delay(Duration=0.45)
   > Set equipAlpha_5(equipAlpha_5=1)
   n21 = Set equipAlpha_5(execute=@n20.else, equipAlpha_5=0)
   n22 = Branch(execute=@n1.then_7, Condition=Get isMelee())
   > Do Once(Reset=@n23.then)
   > Set equipAlpha_6(equipAlpha_6:double=0) -> (Output_Get:double)
   > Delay(Duration=0.4)
   > Set equipAlpha_6(equipAlpha_6=1)
   n23 = Set equipAlpha_6(execute=@n22.else, equipAlpha_6=0)
   n24 = Branch(execute=@n1.then_8, Condition=@n6)
   Delay(execute=[@n24.then, @n26.then], Duration=0.4)
   > Set equipAlpha_FE_06(equipAlpha_FE_06:double=1) -> (Output_Get:double)
   n25 = Branch(execute=@n1.then_9, Condition=@n6)
   Delay(execute=[@n25.then, @n27.then], Duration=0.45)
   > Set equipAlpha_FE_05(equipAlpha_FE_05:double=1) -> (Output_Get:double)
   n26 = Set equipAlpha_FE_06(equipAlpha_FE_06=0)
   n27 = Set equipAlpha_FE_05(equipAlpha_FE_05=0)
   n28 = Set equipAlpha_FE_04(equipAlpha_FE_04=0)
   n29 = Set equipAlpha_FE_03(equipAlpha_FE_03=0)
   n30 = Set equipAlpha_FE_02(equipAlpha_FE_02=0)
   n31 = Set equipAlpha_FE_01(equipAlpha_FE_01=0)
   n32 = Set equipAlpha_FE(equipAlpha_FE=0)
   TryGetPawnOwner()
   RetriggerableDelay(Duration=2)
   RetriggerableDelay(Duration=2)
   RetriggerableDelay(Duration=2)
   RetriggerableDelay(Duration=2)
   RetriggerableDelay(Duration=2)
   RetriggerableDelay(Duration=2)
   RetriggerableDelay(Duration=2)
   n33 = GetOwnerSkeletalMeshComponent()
   n34 = Get WeaponCoreAnim(self=Cast To FortPlayerAnimInstance(Object=GetAnimInstance(self=@n33)))
   Branch(Condition="Equal (Byte)"(A:Byte=Set meleeWeapon(meleeWeapon:Byte=@n34) -> (Output_Get:Byte=0), B:Byte) -> (ReturnValue:bool=~) {is=title K2Node_PromotableOperator})
   "Equal (Float)"(A:Wildcard, B:Wildcard)
  ]}
]
variables [
 Current Weapon {type=FortWeapon, default=null}
 equipAlpha {type=number}
 equipAlpha_1 {type=number}
 equipAlpha_2 {type=number}
 equipAlpha_3 {type=number}
 equipAlpha_4 {type=number}
 equipAlpha_5 {type=number}
 equipAlpha_6 {type=number}
 equipAlpha_FE {type=number}
 equipAlpha_FE_01 {type=number}
 equipAlpha_FE_02 {type=number}
 equipAlpha_FE_03 {type=number}
 equipAlpha_FE_04 {type=number}
 equipAlpha_FE_05 {type=number}
 equipAlpha_FE_06 {type=number}
 isMelee {type=boolean}
 meleeWeapon {type=integer}
]
